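# Animation Blueprint for the M_MED_SteakSting male medium body cosmetic, following the parent mesh's pose and adjusting its belt pouches.
# It copies the attached parent's pose, then chains additive bone-space offsets, two on dyn_belt_pouch and one on dyn_belt_pouch_right.
# The result runs through the body's Control Rig, which receives the emote-playing state as isEmote.
# The update event stores jumping OR falling in Jump&&Fall, with no visible reader in the AnimGraph.
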AnimBlueprint /BRCosmetics/Characters/Player/Male/Medium/Bodies/M_MED_SteakSting/Meshes/M_MED_SteakSting_AnimBP
target_skeleton=&/M_MED_SteakSting_Skeleton
graphs [
 AnimGraph {nodes [
   n1 = "Transform (Modify) Bone - Bone: dyn_belt_pouch"!(
    ComponentPose=Local To Component(LocalPose=Copy Pose From Mesh() {bUseAttachedParent=true})
    bAlphaBoolEnabled:bool=True
    Translation="-0.470,4.876,-8.873"
    Rotation="0,-2.023,-41.178"
    Scale="1,1,1"
   ) -> (Pose) {
    alphaInputType=Bool
    rotationMode=BMM_Additive
    rotationSpace=BCS_BoneSpace
    translationMode=BMM_Additive
    translationSpace=BCS_BoneSpace
    alphaBoolBlend {blendInTime=0.19, blendOption=CubicInOut, blendOutTime=0.17}
    boneToModify {boneName=dyn_belt_pouch}
    rotation {roll=-41.178, yaw=-2.023}
    translation=<-0.47 4.876 -8.873>
   }
   n2 = "Transform (Modify) Bone - Bone: dyn_belt_pouch"(ComponentPose=@n1, bAlphaBoolEnabled=True, Translation="0,1.5,-3", Rotation="-2.101,-6.179,-19.888", Scale="1,1,1") {
    alphaInputType=Bool
    rotationMode=BMM_Additive
    rotationSpace=BCS_BoneSpace
    translationMode=BMM_Additive
    translationSpace=BCS_BoneSpace
    alphaBoolBlend {blendInTime=0.3, blendOption=Linear, blendOutTime=0.3}
    boneToModify {boneName=dyn_belt_pouch}
    rotation=<rot -2.101 -6.179 -19.888>
    translation {y=1.5, z=-3}
   }
   n3 = "Transform (Modify) Bone - Bone: dyn_belt_pouch_right"!(ComponentPose=@n2, bAlphaBoolEnabled:bool=True, Translation="0,0,3", Rotation="0,0,10", Scale="1,1,1") -> (Pose) {
    alphaInputType=Bool
    rotationMode=BMM_Additive
    rotationSpace=BCS_BoneSpace
    translationMode=BMM_Additive
    translationSpace=BCS_BoneSpace
    alphaBoolBlend {blendInTime=0.3, blendOption=Linear, blendOutTime=0.3}
    boneToModify {boneName=dyn_belt_pouch_right}
    rotation {roll=10}
    translation {z=3}
   }
   n4 = Control Rig!(Source=Component To Local(ComponentPose=@n3), Alpha, isEmote:bool=Get bIsPlayingEmote()) -> (Pose) {
    alphaBoolBlend {blendOption=Linear}
    controlRigAssetReference {blueprintRigClass=&/M_MED_SteakSting_CtrlRig.M_MED_SteakSting_CtrlRig_C}
   }
   Output Pose(Result=@n4) {layerGroup=None}
  ]}
 EventGraph {nodes [
   Event BlueprintUpdateAnimation()
   > "Set Jump&&Fall"("Jump&&Fall":bool=OR Boolean(A=Get bIsJumping(), B=Get bIsFalling())) -> (Output_Get:bool)
  ]}
]
variables ["Jump&&Fall" {type=boolean}]
